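# Item-layer AnimBlueprint for Realm player pawns, adding an upper-body deflect animation and overriding other item layers.
# The pawn's IsCurrentlyDeflecting drives a state machine between cached pose and combo loop; on release, the SwingPhase curve selects an inertialized exit animation.
# The older guard/deflect state chain remains but is unreachable from the entry.
# Jump, fall, jetpack and zipline additives switch on bIsMeleeGuarding; skeletal control applies curve-weighted root rotation offset.
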AnimBlueprint /Realm/Anims/Galileo_Realm_LobsterLayerAnimBP
target_skeleton=&Fortnite_M_Avg_Player_Skeleton
graphs [
 AnimGraph {nodes [Output Pose() {layerGroup=None}]}
 EventGraph {nodes [
   Event BlueprintUpdateAnimation()
   > n1 = Cast To BP_PlayerPawn_Realm!(execute, Object=TryGetPawnOwner()) -> (then, CastFailed, AsBP Player Pawn Realm:BP Player Pawn Realm Object Reference) {is=title K2Node_DynamicCast}
   > Set IsCurrentlyDeflecting(IsCurrentlyDeflecting:bool=Get IsCurrentlyDeflecting!(self:BP Player Pawn Realm Object Reference=@n1) -> (IsCurrentlyDeflecting:bool)) -> (Output_Get:bool)
   > Set DeflectPosition(DeflectPosition:double=GetCurveValue(CurveName=SwingPhase)) -> (Output_Get:double)
  ]}
 ItemAimOffset {nodes [
   Input Pose!() -> (Pose, AimOffsetAlpha:double, Yaw:double, Pitch:double) {name=InPose_AimOffset}
   Input Pose!() -> (Pose) {name=InPose_PreAimOffsets}
   Output Pose() {layerGroup=Item Override}
  ]}
 ItemFallAdditive {nodes [
   Input Pose() {name=InPoseItemFallAdditive}
   n1 = Get FallPlayRate() -> (FallPlayRate:float)
   Output Pose(Result=Blend Poses by bool(
     BlendPose_0="Sequence Player (additive) 'M_JumpFall_ARDS_Additive'"!(PlayRate:float=@n1) -> (Pose) {bLoopAnimation=false, sequence=/Game/Animation/Game/MainPlayer/Locomotion/Medium/Male/Jump/AssaultRifle/M_JumpFall_ARDS_Additive}
     BlendPose_1="Sequence Player (additive) 'JumpFall_NoWep_CMM_Additive'"!(PlayRate:float=@n1) -> (Pose) {bLoopAnimation=false, sequence=/Game/Animation/Game/MainPlayer/Locomotion/Medium/Male/Jump/NoWep/JumpFall_NoWep_CMM_Additive}
     bActiveValue=Get bIsMeleeGuarding() -> (bIsMeleeGuarding:bool)
    ) {blendType=Linear}) {layerGroup=Item Override}
  ]}
 ItemFallLandAdditive {nodes [
   Input Pose() {name=InPoseItemFallLandAdditive}
   Output Pose() {layerGroup=Item Override}
  ]}
 ItemFallLevel2Override {nodes [
   Input Pose() {name=InPose_FallLevel2}
   Output Pose() {layerGroup=Item Override}
  ]}
 ItemFinalPoseOverride {nodes [
   Input Pose() {name=InFinalPose}
   Output Pose() {layerGroup=Item Override}
  ]}
 ItemFullBodyAimOffsets {nodes [
   Input Pose!() -> (Pose, FullBody_AimOffsetAlpha:double, Aim_Yaw:double, Aim_Pitch:double) {name=InPoseFullBodyAimOffsets}
   Input Pose!() -> (Pose) {name=InPosePreFullBodyAimOffsets}
   Output Pose() {layerGroup=Item Override}
  ]}
 ItemFullBodyOverride {nodes [
   Input Pose() {name=InPoseFullBody}
   Output Pose() {layerGroup=Item Override}
  ]}
 ItemIdleOverride {nodes [
   Input Pose() {name=InPose_Idle}
   Output Pose() {layerGroup=Item Override}
  ]}
 ItemJetPackJumpAdditive {nodes [Output Pose(Result=Blend Poses by bool(
     BlendPose_0="Sequence Player (additive) 'JetpackApex_ARDS_Loop_Additive'"() {sequence=/Game/Animation/Game/MainPlayer/Locomotion/Medium/Male/Jump/AssaultRifle/JetpackApex_ARDS_Loop_Additive}
     BlendPose_1="Sequence Player (additive) 'JetpackApex_NoWep__Loop_CMM_Additive'"() {sequence=/Game/Animation/Game/MainPlayer/Locomotion/Medium/Male/Jump/NoWep/JetpackApex_NoWep__Loop_CMM_Additive}
     BlendTime_1=1
     bActiveValue=Get bIsMeleeGuarding()
    ) {blendType=Linear}) {layerGroup=Item Override}]}
 ItemJetPackStartAdditive {nodes [Output Pose() {layerGroup=Item Override}]}
 ItemJumpLoopAdditive {nodes [
   Input Pose() {name=InPoseItemJumpLoopAdditive}
   Output Pose(Result=Blend Poses by bool(
     BlendPose_0="Sequence Player (additive) 'M_JumpApex_ARDS_Loop_Additive'"() {playRate=0.5, sequence=/Game/Animation/Game/MainPlayer/Locomotion/Medium/Male/Jump/AssaultRifle/M_JumpApex_ARDS_Loop_Additive}
     BlendPose_1="Sequence Player (additive) 'JumpApex_NoWep__Loop_CMM_Additive'"() {sequence=/Game/Animation/Game/MainPlayer/Locomotion/Medium/Male/Jump/NoWep/JumpApex_NoWep__Loop_CMM_Additive}
     bActiveValue=Get bIsMeleeGuarding()
    ) {blendType=Linear}) {layerGroup=Item Override}
  ]}
 ItemJumpOverride {nodes [
   Input Pose() {name=InPoseJump}
   Output Pose() {layerGroup=Item Override}
  ]}
 ItemJumpUpAdditive {nodes [
   Input Pose() {name=InPoseItemJumpUpAdditive}
   Output Pose(Result=Blend Poses by bool(
     BlendPose_0="Sequence Player (additive) 'M_JumpUp_ARDS_Additive'"() {bLoopAnimation=false, sequence=/Game/Animation/Game/MainPlayer/Locomotion/Medium/Male/Jump/AssaultRifle/M_JumpUp_ARDS_Additive}
     BlendPose_1="Sequence Player (additive) 'JumpUp_NoWep_CMM_Additive'"() {bLoopAnimation=false, sequence=/Game/Animation/Game/MainPlayer/Locomotion/Medium/Male/Jump/NoWep/JumpUp_NoWep_CMM_Additive}
     bActiveValue=Get bIsMeleeGuarding()
    ) {blendType=Linear}) {layerGroup=Item Override}
  ]}
 ItemPreAimOffsets {nodes [
   Input Pose() {name=InPosePreAimOffsets}
   Output Pose() {layerGroup=Item Override}
  ]}
 ItemPreIK {nodes [
   Input Pose() {name=InPose_PreIK}
   Output Pose() {layerGroup=Item Override}
  ]}
 ItemRightHandIK {nodes [
   Input Pose() {name=InPosePreRightHandIK}
   Input Pose() {name=InPosePostRightHandIK}
   Output Pose() {layerGroup=Item Override}
  ]}
 ItemSkeletalControl {nodes [
   n1 = Local To Component(LocalPose=Input Pose())
   n2 = Get RootRotationOffset() -> (RootRotationOffset:Rotator)
   Output Pose(Result=Component To Local(ComponentPose="Transform (Modify) Bone - Bone: Root"!(ComponentPose=@n1, AlphaCurveName:Name=EnableAccelYawCorrection, Rotation=@n2) -> (Pose) {alphaInputType=Curve, rotationMode=BMM_Additive, alphaBoolBlend {blendOption=Linear}, boneToModify {boneName=Root}})) {layerGroup=Item Override}
  ]}
 ItemSwimJumpFallLoop {nodes [Output Pose() {layerGroup=Item Override}]}
 ItemSwimJumpSurfaceEnd {nodes [Output Pose() {layerGroup=Item Override}]}
 ItemSwimJumpSurfaceLoop {nodes [Output Pose() {layerGroup=Item Override}]}
 ItemUpperBody {nodes [
   AnimGraphNode_StateMachine_0 = Deflect() {is=title AnimGraphNode_StateMachine, bAllowConduitEntryStates=true}
   Output Pose(Result=@AnimGraphNode_StateMachine_0) {layerGroup=Item Override}
   Save cached pose 'UpperBody_CachePose'(Pose=Input Pose!() -> (Pose, InputParam:Fort Anim Input Adjusted Aim Structure) {name=InPoseUpperBody}) {cacheName=UpperBody_CachePose}
  ]}
 Deflect {
  entry=Conduit
  path=ItemUpperBody.AnimGraphNode_StateMachine_0.Deflect
  ids=id
  nodes [
   AnimStateConduitNode_0 = BackToDefault() {is=title AnimStateConduitNode}
   AnimStateConduitNode_1 = Conduit()
   AnimStateNode_0 = GuardLoop() {is=title AnimStateNode, stateEntered {guid=#1, notifyName=GuardLoopEntered}}
   AnimStateNode_1 = DeflectEntry1() {is=title AnimStateNode}
   AnimStateNode_10 = Deflect() {is=title AnimStateNode}
   AnimStateNode_11 = Defaultold() {is=title AnimStateNode}
   AnimStateNode_12 = GuardToIdle() {is=title AnimStateNode}
   AnimStateNode_13 = Default() {is=title AnimStateNode}
   AnimStateNode_2 = DeflectHold1() {is=title AnimStateNode, stateEntered {guid=#2, notifyName=DeflectHold1Entered}}
   AnimStateNode_3 = DeflectExit1() {is=title AnimStateNode}
   AnimStateNode_4 = DeflectTransition1_2() {is=title AnimStateNode}
   AnimStateNode_5 = GuardIntro() {is=title AnimStateNode}
   AnimStateNode_6 = DeflectHold2() {is=title AnimStateNode, stateEntered {guid=#3, notifyName=DeflectHold2Entered}}
   AnimStateNode_7 = DeflectTransition2_1() {is=title AnimStateNode}
   AnimStateNode_8 = DeflectExit2() {is=title AnimStateNode}
   AnimStateNode_9 = Deflect_to_Default() {is=title AnimStateNode}
   AnimStateTransitionNode_0 = AnimStateTransitionNode() {crossfadeDuration=0.15}
   AnimStateTransitionNode_1 = AnimStateTransitionNode() {bAutomaticRuleBasedOnSequencePlayerInState=true, crossfadeDuration=0.1}
   AnimStateTransitionNode_10 = AnimStateTransitionNode()
   AnimStateTransitionNode_11 = AnimStateTransitionNode()
   AnimStateTransitionNode_12 = AnimStateTransitionNode() {crossfadeDuration=0.1}
   AnimStateTransitionNode_13 = AnimStateTransitionNode() {crossfadeDuration=0.1}
   AnimStateTransitionNode_14 = AnimStateTransitionNode()
   AnimStateTransitionNode_15 = AnimStateTransitionNode() {bAutomaticRuleBasedOnSequencePlayerInState=true}
   AnimStateTransitionNode_16 = AnimStateTransitionNode()
   AnimStateTransitionNode_17 = AnimStateTransitionNode() {crossfadeDuration=0.1}
   AnimStateTransitionNode_18 = AnimStateTransitionNode()
   AnimStateTransitionNode_19 = AnimStateTransitionNode() {bAutomaticRuleBasedOnSequencePlayerInState=true, crossfadeDuration=0.1}
   AnimStateTransitionNode_2 = AnimStateTransitionNode() {bAutomaticRuleBasedOnSequencePlayerInState=true, crossfadeDuration=0.1}
   AnimStateTransitionNode_20 = AnimStateTransitionNode() {bAutomaticRuleBasedOnSequencePlayerInState=true, crossfadeDuration=0.1}
   AnimStateTransitionNode_21 = AnimStateTransitionNode() {bAutomaticRuleBasedOnSequencePlayerInState=true, crossfadeDuration=0.1}
   AnimStateTransitionNode_22 = AnimStateTransitionNode()
   AnimStateTransitionNode_23 = AnimStateTransitionNode()
   AnimStateTransitionNode_24 = AnimStateTransitionNode() {crossfadeDuration=0.35, logicType=TLT_Inertialization}
   AnimStateTransitionNode_25 = AnimStateTransitionNode() {crossfadeDuration=0.05, logicType=TLT_Inertialization}
   AnimStateTransitionNode_26 = AnimStateTransitionNode() {logicType=TLT_Inertialization, priorityOrder=100}
   AnimStateTransitionNode_27 = AnimStateTransitionNode()
   AnimStateTransitionNode_28 = AnimStateTransitionNode() {bAutomaticRuleBasedOnSequencePlayerInState=true}
   AnimStateTransitionNode_29 = AnimStateTransitionNode() {bAutomaticRuleBasedOnSequencePlayerInState=true}
   AnimStateTransitionNode_3 = AnimStateTransitionNode() {bAutomaticRuleBasedOnSequencePlayerInState=true, crossfadeDuration=0.15}
   AnimStateTransitionNode_30 = AnimStateTransitionNode() {logicType=TLT_Inertialization}
   AnimStateTransitionNode_4 = AnimStateTransitionNode() {bAutomaticRuleBasedOnSequencePlayerInState=true, crossfadeDuration=0.1}
   AnimStateTransitionNode_5 = AnimStateTransitionNode() {bAutomaticRuleBasedOnSequencePlayerInState=true, crossfadeDuration=0.1}
   AnimStateTransitionNode_6 = AnimStateTransitionNode()
   AnimStateTransitionNode_7 = AnimStateTransitionNode()
   AnimStateTransitionNode_8 = AnimStateTransitionNode()
   AnimStateTransitionNode_9 = AnimStateTransitionNode()
  ]
  links [
   {
    dst=Defaultold
    rule=ItemUpperBody.AnimGraphNode_StateMachine_0.Deflect.AnimStateTransitionNode_9.Transition
    src=BackToDefault
    via=AnimStateTransitionNode_9
   }
   {
    dst=Default
    rule=ItemUpperBody.AnimGraphNode_StateMachine_0.Deflect.AnimStateTransitionNode_26.Transition
    src=Conduit
    via=AnimStateTransitionNode_26
   }
   {
    dst=Deflect
    rule=ItemUpperBody.AnimGraphNode_StateMachine_0.Deflect.AnimStateTransitionNode_25.Transition
    src=Conduit
    via=AnimStateTransitionNode_25
   }
   {
    dst=Conduit
    rule=ItemUpperBody.AnimGraphNode_StateMachine_0.Deflect.AnimStateTransitionNode_27.Transition
    src=Default
    via=AnimStateTransitionNode_27
   }
   {
    dst=GuardIntro
    rule=ItemUpperBody.AnimGraphNode_StateMachine_0.Deflect.AnimStateTransitionNode_13.Transition
    src=Defaultold
    via=AnimStateTransitionNode_13
   }
   {
    dst=Conduit
    rule=ItemUpperBody.AnimGraphNode_StateMachine_0.Deflect.AnimStateTransitionNode_28.Transition
    src=Deflect
    via=AnimStateTransitionNode_28
   }
   {
    dst=Deflect_to_Default
    rule=ItemUpperBody.AnimGraphNode_StateMachine_0.Deflect.AnimStateTransitionNode_24.Transition
    src=Deflect
    via=AnimStateTransitionNode_24
   }
   {
    dst=BackToDefault
    rule=ItemUpperBody.AnimGraphNode_StateMachine_0.Deflect.AnimStateTransitionNode_6.Transition
    src=DeflectEntry1
    via=AnimStateTransitionNode_6
   }
   {
    dst=DeflectHold1
    rule=ItemUpperBody.AnimGraphNode_StateMachine_0.Deflect.AnimStateTransitionNode_1.AnimationTransitionGraph_0
    src=DeflectEntry1
    via=AnimStateTransitionNode_1
   }
   {
    dst=BackToDefault
    rule=ItemUpperBody.AnimGraphNode_StateMachine_0.Deflect.AnimStateTransitionNode_16.Transition
    src=DeflectExit1
    via=AnimStateTransitionNode_16
   }
   {
    dst=DeflectTransition1_2
    rule=ItemUpperBody.AnimGraphNode_StateMachine_0.Deflect.AnimStateTransitionNode_10.Transition
    src=DeflectExit1
    via=AnimStateTransitionNode_10
   }
   {
    dst=GuardLoop
    rule=ItemUpperBody.AnimGraphNode_StateMachine_0.Deflect.AnimStateTransitionNode_3.AnimationTransitionGraph_0
    src=DeflectExit1
    via=AnimStateTransitionNode_3
   }
   {
    dst=BackToDefault
    rule=ItemUpperBody.AnimGraphNode_StateMachine_0.Deflect.AnimStateTransitionNode_22.Transition
    src=DeflectExit2
    via=AnimStateTransitionNode_22
   }
   {
    dst=DeflectTransition2_1
    rule=ItemUpperBody.AnimGraphNode_StateMachine_0.Deflect.AnimStateTransitionNode_23.Transition
    src=DeflectExit2
    via=AnimStateTransitionNode_23
   }
   {
    dst=GuardLoop
    rule=ItemUpperBody.AnimGraphNode_StateMachine_0.Deflect.AnimStateTransitionNode_21.Transition
    src=DeflectExit2
    via=AnimStateTransitionNode_21
   }
   {
    dst=BackToDefault
    rule=ItemUpperBody.AnimGraphNode_StateMachine_0.Deflect.AnimStateTransitionNode_7.Transition
    src=DeflectHold1
    via=AnimStateTransitionNode_7
   }
   {
    dst=DeflectExit1
    rule=ItemUpperBody.AnimGraphNode_StateMachine_0.Deflect.AnimStateTransitionNode_2.AnimationTransitionGraph_0
    src=DeflectHold1
    via=AnimStateTransitionNode_2
   }
   {
    dst=DeflectTransition1_2
    rule=ItemUpperBody.AnimGraphNode_StateMachine_0.Deflect.AnimStateTransitionNode_12.Transition
    src=DeflectHold1
    via=AnimStateTransitionNode_12
   }
   {
    dst=BackToDefault
    rule=ItemUpperBody.AnimGraphNode_StateMachine_0.Deflect.AnimStateTransitionNode_11.Transition
    src=DeflectHold2
    via=AnimStateTransitionNode_11
   }
   {
    dst=DeflectExit2
    rule=ItemUpperBody.AnimGraphNode_StateMachine_0.Deflect.AnimStateTransitionNode_20.Transition
    src=DeflectHold2
    via=AnimStateTransitionNode_20
   }
   {
    dst=DeflectTransition2_1
    rule=ItemUpperBody.AnimGraphNode_StateMachine_0.Deflect.AnimStateTransitionNode_17.Transition
    src=DeflectHold2
    via=AnimStateTransitionNode_17
   }
   {
    dst=BackToDefault
    rule=ItemUpperBody.AnimGraphNode_StateMachine_0.Deflect.AnimStateTransitionNode_8.Transition
    src=DeflectTransition1_2
    via=AnimStateTransitionNode_8
   }
   {
    dst=DeflectHold2
    rule=ItemUpperBody.AnimGraphNode_StateMachine_0.Deflect.AnimStateTransitionNode_4.Transition
    src=DeflectTransition1_2
    via=AnimStateTransitionNode_4
   }
   {
    dst=BackToDefault
    rule=ItemUpperBody.AnimGraphNode_StateMachine_0.Deflect.AnimStateTransitionNode_18.Transition
    src=DeflectTransition2_1
    via=AnimStateTransitionNode_18
   }
   {
    dst=DeflectHold1
    rule=ItemUpperBody.AnimGraphNode_StateMachine_0.Deflect.AnimStateTransitionNode_19.Transition
    src=DeflectTransition2_1
    via=AnimStateTransitionNode_19
   }
   {
    dst=Conduit
    rule=ItemUpperBody.AnimGraphNode_StateMachine_0.Deflect.AnimStateTransitionNode_29.Transition
    src=Deflect_to_Default
    via=AnimStateTransitionNode_29
   }
   {
    dst=Deflect
    rule=ItemUpperBody.AnimGraphNode_StateMachine_0.Deflect.AnimStateTransitionNode_30.Transition
    src=Deflect_to_Default
    via=AnimStateTransitionNode_30
   }
   {
    dst=GuardLoop
    rule=ItemUpperBody.AnimGraphNode_StateMachine_0.Deflect.AnimStateTransitionNode_5.Transition
    src=GuardIntro
    via=AnimStateTransitionNode_5
   }
   {
    dst=DeflectEntry1
    rule=ItemUpperBody.AnimGraphNode_StateMachine_0.Deflect.AnimStateTransitionNode_0.AnimationTransitionGraph_0
    src=GuardLoop
    via=AnimStateTransitionNode_0
   }
   {
    dst=GuardToIdle
    rule=ItemUpperBody.AnimGraphNode_StateMachine_0.Deflect.AnimStateTransitionNode_14.Transition
    src=GuardLoop
    via=AnimStateTransitionNode_14
   }
   {
    dst=Defaultold
    rule=ItemUpperBody.AnimGraphNode_StateMachine_0.Deflect.AnimStateTransitionNode_15.Transition
    src=GuardToIdle
    via=AnimStateTransitionNode_15
   }
  ]
 }
 BackToDefault {
  path=ItemUpperBody.AnimGraphNode_StateMachine_0.Deflect.AnimStateConduitNode_0.BackToDefault
  nodes [Result(bCanEnterTransition=NOT Boolean(A=Get bIsMeleeGuarding()))]
 }
 Conduit {path=ItemUpperBody.AnimGraphNode_StateMachine_0.Deflect.AnimStateConduitNode_1.Conduit, nodes [Result(bCanEnterTransition=true)]}
 GuardLoop {
  path=ItemUpperBody.AnimGraphNode_StateMachine_0.Deflect.AnimStateNode_0.GuardLoop
  nodes [
   Sequence Player 'Melee_Pose_Block_CMM'() {sequence=/Game/Animation/Game/MainPlayer/Combat/Melee/Generic/SwordBlock/Melee_Pose_Block_CMM}
   Output Animation Pose(Result=Sequence Player 'Realm_NonTargetedPose'() {sequence=&/Realm_NonTargetedPose})
  ]
 }
 DeflectEntry1 {
  path=ItemUpperBody.AnimGraphNode_StateMachine_0.Deflect.AnimStateNode_1.DeflectEntry1
  nodes [Output Animation Pose(Result=Sequence Player!(Sequence:Anim Sequence Base Object Reference=Get DeflectEntry1Anim() -> (DeflectEntry1Anim:Anim Sequence Object Reference)) -> (Pose) {bLoopAnimation=false})]
 }
 Deflect {
  path=ItemUpperBody.AnimGraphNode_StateMachine_0.Deflect.AnimStateNode_10.Deflect
  nodes [Output Animation Pose(Result=Sequence Player 'Realm_DeflectComboLoop'() {sequence=&/Realm_DeflectComboLoop})]
 }
 Defaultold {
  path=ItemUpperBody.AnimGraphNode_StateMachine_0.Deflect.AnimStateNode_11.Defaultold
  nodes [Output Animation Pose(Result=Use cached pose 'UpperBody_CachePose'())]
 }
 GuardToIdle {
  path=ItemUpperBody.AnimGraphNode_StateMachine_0.Deflect.AnimStateNode_12.GuardToIdle
  nodes [
   Sequence Player 'Melee_Outro_Block_CMM'() {bLoopAnimation=false, sequence=/Game/Animation/Game/MainPlayer/Combat/Melee/Generic/SwordBlock/Melee_Outro_Block_CMM}
   Output Animation Pose(Result=Sequence Player 'Realm_NonTargetedPose'() {sequence=&Realm_NonTargetedPose})
  ]
 }
 Default {
  path=ItemUpperBody.AnimGraphNode_StateMachine_0.Deflect.AnimStateNode_13.Default
  nodes [Output Animation Pose(Result=Use cached pose 'UpperBody_CachePose'())]
 }
 DeflectHold1 {
  path=ItemUpperBody.AnimGraphNode_StateMachine_0.Deflect.AnimStateNode_2.DeflectHold1
  nodes [Output Animation Pose(Result=Sequence Player(Sequence=Get DeflectHold1Anim() -> (DeflectHold1Anim:Anim Sequence Object Reference)) {bLoopAnimation=false})]
 }
 DeflectExit1 {
  path=ItemUpperBody.AnimGraphNode_StateMachine_0.Deflect.AnimStateNode_3.DeflectExit1
  nodes [Output Animation Pose(Result=Sequence Player(Sequence=Get DeflectExit1Anim() -> (DeflectExit1Anim:Anim Sequence Object Reference)) {bLoopAnimation=false})]
 }
 DeflectTransition1_2 {
  path=ItemUpperBody.AnimGraphNode_StateMachine_0.Deflect.AnimStateNode_4.DeflectTransition1_2
  nodes [Output Animation Pose(Result=Sequence Player(Sequence=Get DeflectTransition1_2Anim() -> (DeflectTransition1_2Anim:Anim Sequence Object Reference)) {bLoopAnimation=false})]
 }
 GuardIntro {
  path=ItemUpperBody.AnimGraphNode_StateMachine_0.Deflect.AnimStateNode_5.GuardIntro
  nodes [
   Sequence Player 'Melee_Into_Block_CMM'() {bLoopAnimation=false, sequence=/Game/Animation/Game/MainPlayer/Combat/Melee/Generic/SwordBlock/Melee_Into_Block_CMM}
   Output Animation Pose(Result=Sequence Player 'Realm_NonTargetedPose'() {sequence=&Realm_NonTargetedPose})
  ]
 }
 DeflectHold2 {
  path=ItemUpperBody.AnimGraphNode_StateMachine_0.Deflect.AnimStateNode_6.DeflectHold2
  nodes [Output Animation Pose(Result=Sequence Player(Sequence=Get DeflectHold2Anim() -> (DeflectHold2Anim:Anim Sequence Object Reference)))]
 }
 DeflectTransition2_1 {
  path=ItemUpperBody.AnimGraphNode_StateMachine_0.Deflect.AnimStateNode_7.DeflectTransition2_1
  nodes [Output Animation Pose(Result=Sequence Player(Sequence=Get DeflectTransition2_1Anim() -> (DeflectTransition2_1Anim:Anim Sequence Object Reference)))]
 }
 DeflectExit2 {
  path=ItemUpperBody.AnimGraphNode_StateMachine_0.Deflect.AnimStateNode_8.DeflectExit2
  nodes [Output Animation Pose(Result=Sequence Player(Sequence=Get DeflectExit2Anim() -> (DeflectExit2Anim:Anim Sequence Object Reference)))]
 }
 Deflect_to_Default {
  path=ItemUpperBody.AnimGraphNode_StateMachine_0.Deflect.AnimStateNode_9.Deflect_to_Default
  nodes [
   AnimGraphNode_StateMachine_0 = DeflectExits() {is=title AnimGraphNode_StateMachine, bAllowConduitEntryStates=true}
   Output Animation Pose(Result=@AnimGraphNode_StateMachine_0)
  ]
 }
 DeflectExits {
  entry=Conduit
  path=ItemUpperBody.AnimGraphNode_StateMachine_0.Deflect.AnimStateNode_9.Deflect_to_Default.AnimGraphNode_StateMachine_0.DeflectExits
  ids=id
  nodes [
   AnimStateConduitNode_0 = Conduit()
   AnimStateNode_0 = DeflectExit_1() {is=title AnimStateNode}
   AnimStateNode_1 = DeflectExit_2() {is=title AnimStateNode}
   AnimStateNode_2 = DeflectExit_3() {is=title AnimStateNode}
   AnimStateNode_3 = DeflectExit_4() {is=title AnimStateNode}
   AnimStateNode_4 = DeflectExit_5() {is=title AnimStateNode}
   AnimStateNode_5 = DeflectExit_6() {is=title AnimStateNode}
   AnimStateNode_6 = DeflectExit_7() {is=title AnimStateNode}
   AnimStateNode_7 = DeflectExit_Default() {is=title AnimStateNode}
   AnimStateTransitionNode_0 = AnimStateTransitionNode() {logicType=TLT_Inertialization}
   AnimStateTransitionNode_1 = AnimStateTransitionNode() {logicType=TLT_Inertialization, priorityOrder=2}
   AnimStateTransitionNode_2 = AnimStateTransitionNode() {logicType=TLT_Inertialization, priorityOrder=3}
   AnimStateTransitionNode_3 = AnimStateTransitionNode() {logicType=TLT_Inertialization, priorityOrder=4}
   AnimStateTransitionNode_4 = AnimStateTransitionNode() {logicType=TLT_Inertialization, priorityOrder=5}
   AnimStateTransitionNode_5 = AnimStateTransitionNode() {logicType=TLT_Inertialization, priorityOrder=6}
   AnimStateTransitionNode_6 = AnimStateTransitionNode() {logicType=TLT_Inertialization, priorityOrder=7}
   AnimStateTransitionNode_7 = AnimStateTransitionNode() {logicType=TLT_Inertialization, priorityOrder=100}
  ]
  links [
   {
    dst=DeflectExit_1
    rule=ItemUpperBody.AnimGraphNode_StateMachine_0.Deflect.AnimStateNode_9.Deflect_to_Default.AnimGraphNode_StateMachine_0.DeflectExits.AnimStateTransitionNode_0.Transition
    src=Conduit
    via=AnimStateTransitionNode_0
   }
   {
    dst=DeflectExit_2
    rule=ItemUpperBody.AnimGraphNode_StateMachine_0.Deflect.AnimStateNode_9.Deflect_to_Default.AnimGraphNode_StateMachine_0.DeflectExits.AnimStateTransitionNode_1.Transition
    src=Conduit
    via=AnimStateTransitionNode_1
   }
   {
    dst=DeflectExit_3
    rule=ItemUpperBody.AnimGraphNode_StateMachine_0.Deflect.AnimStateNode_9.Deflect_to_Default.AnimGraphNode_StateMachine_0.DeflectExits.AnimStateTransitionNode_2.Transition
    src=Conduit
    via=AnimStateTransitionNode_2
   }
   {
    dst=DeflectExit_4
    rule=ItemUpperBody.AnimGraphNode_StateMachine_0.Deflect.AnimStateNode_9.Deflect_to_Default.AnimGraphNode_StateMachine_0.DeflectExits.AnimStateTransitionNode_3.Transition
    src=Conduit
    via=AnimStateTransitionNode_3
   }
   {
    dst=DeflectExit_5
    rule=ItemUpperBody.AnimGraphNode_StateMachine_0.Deflect.AnimStateNode_9.Deflect_to_Default.AnimGraphNode_StateMachine_0.DeflectExits.AnimStateTransitionNode_4.Transition
    src=Conduit
    via=AnimStateTransitionNode_4
   }
   {
    dst=DeflectExit_6
    rule=ItemUpperBody.AnimGraphNode_StateMachine_0.Deflect.AnimStateNode_9.Deflect_to_Default.AnimGraphNode_StateMachine_0.DeflectExits.AnimStateTransitionNode_5.Transition
    src=Conduit
    via=AnimStateTransitionNode_5
   }
   {
    dst=DeflectExit_7
    rule=ItemUpperBody.AnimGraphNode_StateMachine_0.Deflect.AnimStateNode_9.Deflect_to_Default.AnimGraphNode_StateMachine_0.DeflectExits.AnimStateTransitionNode_6.Transition
    src=Conduit
    via=AnimStateTransitionNode_6
   }
   {
    dst=DeflectExit_Default
    rule=ItemUpperBody.AnimGraphNode_StateMachine_0.Deflect.AnimStateNode_9.Deflect_to_Default.AnimGraphNode_StateMachine_0.DeflectExits.AnimStateTransitionNode_7.Transition
    src=Conduit
    via=AnimStateTransitionNode_7
   }
  ]
 }
 Conduit {
  path=ItemUpperBody.AnimGraphNode_StateMachine_0.Deflect.AnimStateNode_9.Deflect_to_Default.AnimGraphNode_StateMachine_0.DeflectExits.AnimStateConduitNode_0.Conduit
  nodes [Result(bCanEnterTransition=true)]
 }
 DeflectExit_1 {
  path=ItemUpperBody.AnimGraphNode_StateMachine_0.Deflect.AnimStateNode_9.Deflect_to_Default.AnimGraphNode_StateMachine_0.DeflectExits.AnimStateNode_0.DeflectExit_1
  nodes [Output Animation Pose(Result=Sequence Player 'Realm_Deflect1_Exit'() {bLoopAnimation=false, sequence=&/Realm_Deflect1_Exit})]
 }
 DeflectExit_2 {
  path=ItemUpperBody.AnimGraphNode_StateMachine_0.Deflect.AnimStateNode_9.Deflect_to_Default.AnimGraphNode_StateMachine_0.DeflectExits.AnimStateNode_1.DeflectExit_2
  nodes [Output Animation Pose(Result=Sequence Player 'Realm_Deflect2_Exit'() {bLoopAnimation=false, sequence=&/Realm_Deflect2_Exit})]
 }
 DeflectExit_3 {
  path=ItemUpperBody.AnimGraphNode_StateMachine_0.Deflect.AnimStateNode_9.Deflect_to_Default.AnimGraphNode_StateMachine_0.DeflectExits.AnimStateNode_2.DeflectExit_3
  nodes [Output Animation Pose(Result=Sequence Player 'Realm_Deflect3_Exit'() {bLoopAnimation=false, sequence=&/Realm_Deflect3_Exit})]
 }
 DeflectExit_4 {
  path=ItemUpperBody.AnimGraphNode_StateMachine_0.Deflect.AnimStateNode_9.Deflect_to_Default.AnimGraphNode_StateMachine_0.DeflectExits.AnimStateNode_3.DeflectExit_4
  nodes [Output Animation Pose(Result=Sequence Player 'Realm_Deflect4_Exit'() {bLoopAnimation=false, sequence=&/Realm_Deflect4_Exit})]
 }
 DeflectExit_5 {
  path=ItemUpperBody.AnimGraphNode_StateMachine_0.Deflect.AnimStateNode_9.Deflect_to_Default.AnimGraphNode_StateMachine_0.DeflectExits.AnimStateNode_4.DeflectExit_5
  nodes [Output Animation Pose(Result=Sequence Player 'Realm_Deflect5_Exit'() {bLoopAnimation=false, sequence=&/Realm_Deflect5_Exit})]
 }
 DeflectExit_6 {
  path=ItemUpperBody.AnimGraphNode_StateMachine_0.Deflect.AnimStateNode_9.Deflect_to_Default.AnimGraphNode_StateMachine_0.DeflectExits.AnimStateNode_5.DeflectExit_6
  nodes [Output Animation Pose(Result=Sequence Player 'Realm_Deflect6_Exit'() {bLoopAnimation=false, sequence=&/Realm_Deflect6_Exit})]
 }
 DeflectExit_7 {
  path=ItemUpperBody.AnimGraphNode_StateMachine_0.Deflect.AnimStateNode_9.Deflect_to_Default.AnimGraphNode_StateMachine_0.DeflectExits.AnimStateNode_6.DeflectExit_7
  nodes [Output Animation Pose(Result=Sequence Player 'Realm_Deflect7_Exit'() {bLoopAnimation=false, sequence=&/Realm_Deflect7_Exit})]
 }
 DeflectExit_Default {
  path=ItemUpperBody.AnimGraphNode_StateMachine_0.Deflect.AnimStateNode_9.Deflect_to_Default.AnimGraphNode_StateMachine_0.DeflectExits.AnimStateNode_7.DeflectExit_Default
  nodes [Output Animation Pose(Result=Sequence Player 'Realm_Deflect8_Exit'() {bLoopAnimation=false, sequence=&/Realm_Deflect8_Exit})]
 }
 Transition {
  path=ItemUpperBody.AnimGraphNode_StateMachine_0.Deflect.AnimStateNode_9.Deflect_to_Default.AnimGraphNode_StateMachine_0.DeflectExits.AnimStateTransitionNode_0.Transition
  nodes [
   n1 = Get DeflectPosition() -> (DeflectPosition:double)
   Result(bCanEnterTransition=AND Boolean(A="float < float"(A=@n1, B=0.2), B="float >= float"(A=@n1, B=0.1)))
  ]
 }
 Transition {
  path=ItemUpperBody.AnimGraphNode_StateMachine_0.Deflect.AnimStateNode_9.Deflect_to_Default.AnimGraphNode_StateMachine_0.DeflectExits.AnimStateTransitionNode_1.Transition
  nodes [Result(bCanEnterTransition="float < float"(A=Get DeflectPosition(), B=0.36))]
 }
 Transition {
  path=ItemUpperBody.AnimGraphNode_StateMachine_0.Deflect.AnimStateNode_9.Deflect_to_Default.AnimGraphNode_StateMachine_0.DeflectExits.AnimStateTransitionNode_2.Transition
  nodes [Result(bCanEnterTransition="float < float"(A=Get DeflectPosition(), B=0.5))]
 }
 Transition {
  path=ItemUpperBody.AnimGraphNode_StateMachine_0.Deflect.AnimStateNode_9.Deflect_to_Default.AnimGraphNode_StateMachine_0.DeflectExits.AnimStateTransitionNode_3.Transition
  nodes [Result(bCanEnterTransition="float < float"(A=Get DeflectPosition(), B=0.58))]
 }
 Transition {
  path=ItemUpperBody.AnimGraphNode_StateMachine_0.Deflect.AnimStateNode_9.Deflect_to_Default.AnimGraphNode_StateMachine_0.DeflectExits.AnimStateTransitionNode_4.Transition
  nodes [Result(bCanEnterTransition="float < float"(A=Get DeflectPosition(), B=0.76))]
 }
 Transition {
  path=ItemUpperBody.AnimGraphNode_StateMachine_0.Deflect.AnimStateNode_9.Deflect_to_Default.AnimGraphNode_StateMachine_0.DeflectExits.AnimStateTransitionNode_5.Transition
  nodes [Result(bCanEnterTransition="float < float"(A=Get DeflectPosition(), B=0.85))]
 }
 Transition {
  path=ItemUpperBody.AnimGraphNode_StateMachine_0.Deflect.AnimStateNode_9.Deflect_to_Default.AnimGraphNode_StateMachine_0.DeflectExits.AnimStateTransitionNode_6.Transition
  nodes [Result(bCanEnterTransition="float < float"(A=Get DeflectPosition(), B=0.97))]
 }
 Transition {
  path=ItemUpperBody.AnimGraphNode_StateMachine_0.Deflect.AnimStateNode_9.Deflect_to_Default.AnimGraphNode_StateMachine_0.DeflectExits.AnimStateTransitionNode_7.Transition
  nodes [Result(bCanEnterTransition=true)]
 }
 AnimationTransitionGraph_0 {
  path=ItemUpperBody.AnimGraphNode_StateMachine_0.Deflect.AnimStateTransitionNode_0.AnimationTransitionGraph_0
  nodes [Result(bCanEnterTransition=Get bIsMeleeDeflecting() -> (bIsMeleeDeflecting:bool))]
 }
 AnimationTransitionGraph_0 {path=ItemUpperBody.AnimGraphNode_StateMachine_0.Deflect.AnimStateTransitionNode_1.AnimationTransitionGraph_0, nodes [Result()]}
 Transition {
  path=ItemUpperBody.AnimGraphNode_StateMachine_0.Deflect.AnimStateTransitionNode_10.Transition
  nodes [Result(bCanEnterTransition=Get bIsMeleeDeflecting())]
 }
 Transition {
  path=ItemUpperBody.AnimGraphNode_StateMachine_0.Deflect.AnimStateTransitionNode_11.Transition
  nodes [Result(bCanEnterTransition=NOT Boolean(A=Get bIsMeleeGuarding()))]
 }
 Transition {
  path=ItemUpperBody.AnimGraphNode_StateMachine_0.Deflect.AnimStateTransitionNode_12.Transition
  nodes [Result(bCanEnterTransition=Get bIsMeleeDeflecting())]
 }
 Transition {
  path=ItemUpperBody.AnimGraphNode_StateMachine_0.Deflect.AnimStateTransitionNode_13.Transition
  nodes [Result(bCanEnterTransition=Get bIsMeleeGuarding())]
 }
 Transition {
  path=ItemUpperBody.AnimGraphNode_StateMachine_0.Deflect.AnimStateTransitionNode_14.Transition
  nodes [Result(bCanEnterTransition=NOT Boolean(A=Get bIsMeleeGuarding()))]
 }
 Transition {path=ItemUpperBody.AnimGraphNode_StateMachine_0.Deflect.AnimStateTransitionNode_15.Transition, nodes [Result()]}
 Transition {
  path=ItemUpperBody.AnimGraphNode_StateMachine_0.Deflect.AnimStateTransitionNode_16.Transition
  nodes [Result(bCanEnterTransition=NOT Boolean(A=Get bIsMeleeGuarding()))]
 }
 Transition {
  path=ItemUpperBody.AnimGraphNode_StateMachine_0.Deflect.AnimStateTransitionNode_17.Transition
  nodes [Result(bCanEnterTransition=Get bIsMeleeDeflecting())]
 }
 Transition {
  path=ItemUpperBody.AnimGraphNode_StateMachine_0.Deflect.AnimStateTransitionNode_18.Transition
  nodes [Result(bCanEnterTransition=NOT Boolean(A=Get bIsMeleeGuarding()))]
 }
 Transition {path=ItemUpperBody.AnimGraphNode_StateMachine_0.Deflect.AnimStateTransitionNode_19.Transition, nodes [Result()]}
 AnimationTransitionGraph_0 {path=ItemUpperBody.AnimGraphNode_StateMachine_0.Deflect.AnimStateTransitionNode_2.AnimationTransitionGraph_0, nodes [Result()]}
 Transition {path=ItemUpperBody.AnimGraphNode_StateMachine_0.Deflect.AnimStateTransitionNode_20.Transition, nodes [Result()]}
 Transition {path=ItemUpperBody.AnimGraphNode_StateMachine_0.Deflect.AnimStateTransitionNode_21.Transition, nodes [Result()]}
 Transition {
  path=ItemUpperBody.AnimGraphNode_StateMachine_0.Deflect.AnimStateTransitionNode_22.Transition
  nodes [Result(bCanEnterTransition=NOT Boolean(A=Get bIsMeleeGuarding()))]
 }
 Transition {
  path=ItemUpperBody.AnimGraphNode_StateMachine_0.Deflect.AnimStateTransitionNode_23.Transition
  nodes [Result(bCanEnterTransition=Get bIsMeleeDeflecting())]
 }
 Transition {
  path=ItemUpperBody.AnimGraphNode_StateMachine_0.Deflect.AnimStateTransitionNode_24.Transition
  nodes [Result(bCanEnterTransition=NOT Boolean(A=Get IsCurrentlyDeflecting!() -> (IsCurrentlyDeflecting)))]
 }
 Transition {
  path=ItemUpperBody.AnimGraphNode_StateMachine_0.Deflect.AnimStateTransitionNode_25.Transition
  nodes [Result(bCanEnterTransition=Get IsCurrentlyDeflecting())]
 }
 Transition {path=ItemUpperBody.AnimGraphNode_StateMachine_0.Deflect.AnimStateTransitionNode_26.Transition, nodes [Result(bCanEnterTransition=true)]}
 Transition {path=ItemUpperBody.AnimGraphNode_StateMachine_0.Deflect.AnimStateTransitionNode_27.Transition, nodes [Result(bCanEnterTransition=true)]}
 Transition {path=ItemUpperBody.AnimGraphNode_StateMachine_0.Deflect.AnimStateTransitionNode_28.Transition, nodes [Result(bCanEnterTransition=true)]}
 Transition {path=ItemUpperBody.AnimGraphNode_StateMachine_0.Deflect.AnimStateTransitionNode_29.Transition, nodes [Result()]}
 AnimationTransitionGraph_0 {path=ItemUpperBody.AnimGraphNode_StateMachine_0.Deflect.AnimStateTransitionNode_3.AnimationTransitionGraph_0, nodes [Result()]}
 Transition {
  path=ItemUpperBody.AnimGraphNode_StateMachine_0.Deflect.AnimStateTransitionNode_30.Transition
  nodes [Result(bCanEnterTransition=Get IsCurrentlyDeflecting())]
 }
 Transition {path=ItemUpperBody.AnimGraphNode_StateMachine_0.Deflect.AnimStateTransitionNode_4.Transition, nodes [Result()]}
 Transition {path=ItemUpperBody.AnimGraphNode_StateMachine_0.Deflect.AnimStateTransitionNode_5.Transition, nodes [Result()]}
 Transition {
  path=ItemUpperBody.AnimGraphNode_StateMachine_0.Deflect.AnimStateTransitionNode_6.Transition
  nodes [Result(bCanEnterTransition=NOT Boolean(A=Get bIsMeleeGuarding()))]
 }
 Transition {
  path=ItemUpperBody.AnimGraphNode_StateMachine_0.Deflect.AnimStateTransitionNode_7.Transition
  nodes [Result(bCanEnterTransition=NOT Boolean(A=Get bIsMeleeGuarding()))]
 }
 Transition {
  path=ItemUpperBody.AnimGraphNode_StateMachine_0.Deflect.AnimStateTransitionNode_8.Transition
  nodes [Result(bCanEnterTransition=NOT Boolean(A=Get bIsMeleeGuarding()))]
 }
 Transition {
  path=ItemUpperBody.AnimGraphNode_StateMachine_0.Deflect.AnimStateTransitionNode_9.Transition
  nodes [Result(bCanEnterTransition=NOT Boolean(A=Get bIsMeleeGuarding()))]
 }
 ItemZipLineStartAdditive {nodes [Output Pose(Result=Blend Poses by bool(
     BlendPose_0="Sequence Player (additive) 'ZipLine_Entry_ARDS_Additive'"() {bLoopAnimation=false, sequence=/Game/Animation/Game/MainPlayer/Combat/Gadgets/Medium/Male/ZipLine/ZipLine_Entry_ARDS_Additive}
     BlendPose_1="Sequence Player (additive) 'ZipLine_Entry_NoWep_Additive'"() {bLoopAnimation=false, sequence=/Game/Animation/Game/MainPlayer/Combat/Gadgets/Medium/Male/ZipLine/ZipLine_Entry_NoWep_Additive}
     bActiveValue=Get bIsMeleeGuarding()
    ) {blendType=Linear}) {layerGroup=Item Override}]}
]
variables [
 DeflectPosition {type=number}
 IsCurrentlyDeflecting {type=boolean}
]
