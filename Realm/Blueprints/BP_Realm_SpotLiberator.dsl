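; BP_Realm_SpotLiberator is a one-shot actor that frees reserved Living World density at BP_Realm_SpotReservation actors and triggers event generation.
; On BeginPlay it waits for playlist data, then validates and stores the FortAthenaLivingWorldManager component from the GameState.
; It calls ReleaseActorDensity on every spot reservation actor, requests event generation from the manager, and destroys itself after a short delay.
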(event EventBeginPlay
  (bind _returnvalue (Game|GetGameState))
  (bind _returnvalue_2 (Actor|GetComponentbyClass _returnvalue "/Script/LagerRuntime.FortAthenaLivingWorldManager"))
  (Class|FortAsyncActionCurrentPlaylistReady|PlaylistDataReadyAsync
    (:OnReady
      (Utilities|IsValid _returnvalue_2
        (:Is%20Valid
          (Variables|Default|SetLivingWorldManager _returnvalue_2)
          (bind _outactors (Actor|GetAllActorsOfClass "/Realm/Blueprints/BP_Realm_SpotReservation.BP_Realm_SpotReservation_C"))
          (for _array_element _outactors
            (Class|BPRealmSpotReservation|ReleaseActorDensity _array_element))
          (Class|FortAthenaLivingWorldManager|RequestEventGeneration (Variables|Default|GetLivingWorldManager))
          (Utilities|FlowControl|Delay 0.2)
          (Actor|DestroyActor))))))

(event Collision|EventActorBeginOverlap
  (params (OtherActor :type Actor%20Object%20Reference)))

(event EventTick
  (params (DeltaSeconds :type Float%20%28single-precision%29)))

(fn ConstructionScript)
