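; Realm spot actor that reserves AI density with Fortnite's Living World Manager so animals are not spawned around it.
; On BeginPlay it gets the manager component from the GameState, and once playlist data is ready caches it and registers itself as a density reservoir.
; A custom ReleaseActorDensity event unregisters the reservoir, then destroys the actor after a short delay.
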(event EventBeginPlay
  (bind _returnvalue (Game|GetGameState))
  (bind _returnvalue_2 (Actor|GetComponentbyClass _returnvalue "/Script/LagerRuntime.FortAthenaLivingWorldManager"))
  (Class|FortAsyncActionCurrentPlaylistReady|PlaylistDataReadyAsync
    (:OnReady
      (Utilities|IsValid _returnvalue_2
        (:Is%20Valid
          (bind _output_get (Variables|Default|SetLivingWorldManager _returnvalue_2))
          (Class|FortAthenaLivingWorldManager|RegisterDensityReservoir _output_get self 1.0 1.5e+04) ; Tell LWM we will be spawning 3 AI here so they don't try to spawn Animals around that could cause issues with Density
        )))))

(event Custom|ReleaseActorDensity
  (bind _livingworldmanager (Variables|Default|GetLivingWorldManager))
  (Utilities|IsValid _livingworldmanager
    (:Is%20Valid
      (Class|FortAthenaLivingWorldManager|UnregisterDensityReservoir _livingworldmanager self)
      (Utilities|FlowControl|Delay 0.2)
      (Actor|DestroyActor))))

(event Collision|EventActorBeginOverlap
  (params (OtherActor :type Actor%20Object%20Reference)))

(event EventTick
  (params (DeltaSeconds :type Float%20%28single-precision%29)))

(fn ConstructionScript)
